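; InputsFallback_Empty is a PCG graph, named as an empty fallback for inputs in a set of shared primitives.
; It declares one string parameter, WarningMessage, with a default message saying that no valid input data was found.
; The graph shows no nodes or connections, so nothing in it reads or emits that message.
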(graph InputsFallback_Empty
  (params (WarningMessage :type String :default "No Valid Input Data Found")))
